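# Sound cue for a transformer trap's electrical spark break effect in the Fort_Traps sound class.
# A Mixer layers a fixed ElectricitySpark_Zap_2 wave with a Random pick between two equally weighted electric ceiling trap spark waves.
# The cue plays pitched up, uses Fort_Traps attenuation, and overrides concurrency to prevent new instances at the limit.
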SoundCue /Game/Sounds/Fort_Traps/Transformer/ElectricitySpark_Break_SoundCue
root=Mixer_17
nodes [
 {id=Mixer_17, children [Random_32, WavePlayer_1934], properties {inputVolume [1,1]}}
 {id=Random_32, children [WavePlayer_114, WavePlayer_119], properties {weights [1,1]}}
 {id=WavePlayer_114, properties {soundWaveAssetPtr=/Game/Sounds/Fort_Traps/Electric_Trap/CeilingTrap_Electric_Spark_007}}
 {id=WavePlayer_119, properties {soundWaveAssetPtr=/Game/Sounds/Fort_Traps/Electric_Trap/CeilingTrap_Electric_Spark_002}}
 {id=WavePlayer_1934, properties {soundWaveAssetPtr=&/ElectricitySpark_Zap_2}}
]
properties {
 attenuationSettings=&Fort_Traps_Att
 bOverrideConcurrency=true
 maxDistance=6000
 pitchMultiplier=1.5
 soundClassObject=&Fort_Traps
 volumeMultiplier=1
 attenuationOverrides {falloffDistance=2600}
 concurrencyOverrides {resolutionRule=PreventNew}
}
