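# Sound cue for an electric spark zap in the Fort_Traps Transformer sound set.
# A quality-level node feeds two mixers, each layering a constant zap wave with a randomly chosen, equally weighted ceiling-trap spark.
# The first quality slot picks from four sparks and the second from three, while the other two slots are empty.
# Playback uses the shared trap sound class, attenuation and a three-voice concurrency rule.
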SoundCue /Game/Sounds/Fort_Traps/Transformer/ElectricitySpark_Zap_SoundCue
root=QualityLevel_0
nodes [
 {id=Mixer_1, children [Random_1,WavePlayer_5], properties {inputVolume [1,1]}}
 {id=Mixer_2, children [Random_2, WavePlayer_10], properties {inputVolume [1,1]}}
 {id=QualityLevel_0, children [Mixer_1,Mixer_2,null,null]}
 {id=Random_1, children [WavePlayer_6,WavePlayer_7,WavePlayer_8,WavePlayer_9], properties {weights [1,1,1,1]}}
 {id=Random_2, children [WavePlayer_11, WavePlayer_12, WavePlayer_14], properties {weights [1,1,1]}}
 {id=WavePlayer_10, properties {soundWaveAssetPtr=&/ElectricitySpark_Zap_2}}
 {id=WavePlayer_11, properties {soundWaveAssetPtr=/Game/Sounds/Fort_Traps/Electric_Trap/CeilingTrap_Electric_Spark_008}}
 {id=WavePlayer_12, properties {soundWaveAssetPtr=/Game/Sounds/Fort_Traps/Electric_Trap/CeilingTrap_Electric_Spark_007}}
 {id=WavePlayer_14, properties {soundWaveAssetPtr=/Game/Sounds/Fort_Traps/Electric_Trap/CeilingTrap_Electric_Spark_002}}
 {id=WavePlayer_5, properties {soundWaveAssetPtr=&ElectricitySpark_Zap_2}}
 {id=WavePlayer_6, properties {soundWaveAssetPtr=&CeilingTrap_Electric_Spark_008}}
 {id=WavePlayer_7, properties {soundWaveAssetPtr=&CeilingTrap_Electric_Spark_007}}
 {id=WavePlayer_8, properties {soundWaveAssetPtr=/Game/Sounds/Fort_Traps/Electric_Trap/CeilingTrap_Electric_Spark_004}}
 {id=WavePlayer_9, properties {soundWaveAssetPtr=&CeilingTrap_Electric_Spark_002}}
]
properties {
 attenuationSettings=&Fort_Traps_Att
 maxDistance=6000
 soundClassObject=&Fort_Traps
 volumeMultiplier=0.5
 attenuationOverrides {falloffDistance=2600}
 concurrencyOverrides {resolutionRule=PreventNew}
 concurrencySet [/Game/Sounds/Concurrency_Rules/ThreeMaximum_StopOldest_VolumeScale]
}
