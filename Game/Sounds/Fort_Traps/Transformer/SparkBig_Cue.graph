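# Sound cue for the big electrical spark of the transformer trap.
# A random node picks one of four equally weighted SparkBig waves and feeds it through a modulator at default settings.
# It plays quietly on the gadgets sound class with its own spark attenuation.
# A concurrency override caps simultaneous instances, stopping the oldest and scaling volume down as copies stack.
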SoundCue /Game/Sounds/Fort_Traps/Transformer/SparkBig_Cue
root=Modulator_0
nodes [
 {id=Modulator_0, children [Random_0]}
 {id=Random_0, children [WavePlayer_1,WavePlayer_2,WavePlayer_3,WavePlayer_4], properties {weights [1,1,1,1]}}
 {id=WavePlayer_1, properties {soundWaveAssetPtr=&/SparkBig_01}}
 {id=WavePlayer_2, properties {soundWaveAssetPtr=&/SparkBig_02}}
 {id=WavePlayer_3, properties {soundWaveAssetPtr=&/SparkBig_03}}
 {id=WavePlayer_4, properties {soundWaveAssetPtr=&/SparkBig_04}}
]
properties {
 attenuationSettings=/Game/Sounds/Attenuation/Fort_SparkBig_Att
 bOverrideConcurrency=true
 maxDistance=3000
 soundClassObject=/Game/Sounds/FortSoundClassesAndModes/Fort_Gadgets_StW
 volumeMultiplier=0.3
 concurrencyOverrides {maxCount=4, resolutionRule=StopOldest, volumeScale=0.7}
}
